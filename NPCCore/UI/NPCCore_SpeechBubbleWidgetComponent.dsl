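; Widget component that displays an NPC's speech bubble on clients, driven by the owner's FortPawnComponent_SpeechBubble.
; On non-authority BeginPlay it binds to that component's show and hide events, then unhides itself and passes the bubble text to UI_NPCCore_SpeechBubble, or hides again.
; While shown, a looping timer raises the widget's render translation with distance to the first local player's pawn, up to a clamped limit.
; EndPlay unbinds and clears the timer.
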(event EventBeginPlay
  (bind _returnvalue (Components|GetOwner))
  (bind _returnvalue_2 (Actor|GetComponentbyClass _returnvalue "/Script/FortniteGame.FortPawnComponent_SpeechBubble"))
  (|CallParentFunction)
  (Utilities|SwitchHasAuthority
    (:Remote
      (Utilities|IsValid _returnvalue_2
        (:Is%20Valid
          (EventDispatchers|BindEventtoSpeechBubbleShow _returnvalue_2 (delegate Custom|OnSpeechBubbleShowEvent))
          (EventDispatchers|BindEventtoSpeechBubbleHide _returnvalue_2 (delegate Custom|OnSpeechBubbleHideEvent)))))))

(event Custom|OnSpeechBubbleShowEvent
  ;[1 1 1] Show SpeechBubble
  (Development|SetHiddeninGame self false false)
  (bind _asui_npccore_speech_bubble (Utilities|Casting|CastToUI_NPCCore_SpeechBubble (UserInterface|GetUserWidgetObject)))
  (Conversation|SetPreviewMessage _asui_npccore_speech_bubble (SpeechBubble|GetSpeechBubbleText (Actor|GetComponentbyClass (Components|GetOwner) "/Script/FortniteGame.FortPawnComponent_SpeechBubble")))
  (bind _returnvalue (Utilities|Time|SetTimerbyEvent (EventDispatchers|CreateEvent :!event Custom|UpdatedWidgetOffset) 0.1 true))
  (Variables|Default|SetUpdateOffsetTimerHandle _returnvalue))

(event Custom|OnSpeechBubbleHideEvent
  ;[1 1 1] Hide SpeechBubble
  (Development|SetHiddeninGame self true false)
  (Utilities|Time|ClearandInvalidateTimerbyHandle (Variables|Default|GetUpdateOffsetTimerHandle)))

(event Custom|UpdatedWidgetOffset
  (bind _returnvalue (Fortnite|GetFirstLocalFortPlayerController true))
  (bind _returnvalue_2 (Pawn|GetControlledPawn _returnvalue))
  ;[1 1 1] UpdateWidgetOffset
  (Utilities|IsValid _returnvalue_2
    (:Is%20Valid
      (Variables|/Script/UMG.Widget|SetRenderTransform_Translation_X (Utilities|Struct|MakeWidgetTransform (Math|Vector2D|MakeVector2D 0.0 (Math|Float|Lerp 0.0 -1e+02 (Math|Float|Clamp(Float) (Math|Float|NormalizetoRange (Math|Vector|Distance(Vector) (Transformation|GetActorLocation (Components|GetOwner)) (Transformation|GetActorLocation _returnvalue_2)) 0.0 5e+03))))) (UserInterface|GetUserWidgetObject)))))

(event EventEndPlay
  (params (EndPlayReason :type EEndPlayReason%20Enum :default "Destroyed"))
  (bind _returnvalue (Components|GetOwner))
  (bind _returnvalue_2 (Actor|GetComponentbyClass _returnvalue "/Script/FortniteGame.FortPawnComponent_SpeechBubble"))
  (|CallParentFunction EndPlayReason)
  (Utilities|IsValid _returnvalue_2
    (:Is%20Valid
      (EventDispatchers|UnbindallEventsfromSpeechBubbleShow _returnvalue_2)
      (EventDispatchers|UnbindallEventsfromSpeechBubbleHide _returnvalue_2)
      (Utilities|Time|ClearandInvalidateTimerbyHandle (Variables|Default|GetUpdateOffsetTimerHandle)))))
